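# Niagara system for the Evil Energy vehicle boost cosmetic, building an exhaust effect at the boost socket.
# A three-renderer ribbon trail combines with smoke, eye and contour sprites, positioned from the single-boost position and car offset; Eye_Single is disabled.
# GameModeID switches adjust offsets and sizes, while painted-variant overrides recolor the layers.
# Contour and eye sprites feed sine and cosine waves into dynamic material parameters for animation.
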NiagaraSystem /VehicleCosmetics/Boosts/Boost_EvilEnergy/FX/NS_EVIL_PROVIS
dependencies {
 usedDataInterfaces [
  &NiagaraDataInterfaceCurve
  &NiagaraDataInterfaceColorCurve
  &NiagaraDataInterfaceSpriteRendererInfo
  &NiagaraDataInterfaceVector2DCurve
  &NiagaraDataInterfaceCurlNoise
  &NiagaraDataInterfaceArrayFloat
  &NiagaraDataInterfaceSkeletalMesh
  &NiagaraDataInterfaceArrayFloat3
 ]
 usedDynamicInputs [&SelectVectorFromArray]
 usedModules [
  &self:SystemSpawnScript.SystemScriptSource.SystemScriptGraph.NiagaraNodeAssignment_0.SetVariables_#1
  &self:SystemSpawnScript.SystemScriptSource.SystemScriptGraph.NiagaraNodeAssignment_3.SetVariables_#2
  &self:SystemSpawnScript.SystemScriptSource.SystemScriptGraph.NiagaraNodeAssignment_2.SetVariables_#3
  &SystemState
 ]
}
emitters {
 Eye_Single {
  bEnabled=false
  emitterUpdateScript [
   &EmitterState(Life Cycle Mode=::NewEnumerator1, Loop Behavior=::NewEnumerator1, Loop Duration Mode=::NewEnumerator1, UseLoopDelay=false)
   &SpawnBurst_Instantaneous(Use Spawn Probability=false, UseLoopCountLimit=false) {enabled=false}
   &SpawnRate(SpawnGroup=1, SpawnRate=8, Use Spawn Probability=false) {module=SpawnRate002}
   &SpawnRate(SpawnRate=24, Use Spawn Probability=false)
  ]
  particleSpawnScript [
   /SetVariables(Particles.VisibilityTag=$Emitter.SpawnGroup) {isSetParameters=true, module=SetVariables_#4}
   &VC_PaintedVariantsOverride() {enabled=false}
   &InitializeParticle(
    Lifetime=&MaskFloatBySpawnGroup(
     Masked Float=&MergeHelper_FloatSwitch(0=99999, 1=0.5, 2=0.5, 3=0.5, Index=$System.GameModeID, Num=::NewEnumerator2)
     Passthrough Float=&MergeHelper_FloatSwitch(0=0.2, 1=0.2, 2=0.2, 3=0.2, Index=$System.GameModeID, Num=::NewEnumerator2)
    )
    Position=&AddVectorToPosition(Position=$User.DTCarOffset, Vector=$System.FinalPosition)
    Position Mode=::NewEnumerator1
    Position Offset=&MergeHelper_VectorSwitch(0=<-47.5 0 0>, 1=<-47.5 0 0>, 2=<0 0 50>, 3=<-200 0 37>, Index=$System.GameModeID, Num=::NewEnumerator2)
    Position Offset Coordinate Space=::Local
    Sprite Rotation Mode=::NewEnumerator0
    Sprite Size Mode=::NewEnumerator3
    Uniform Sprite Size=&MaskFloatBySpawnGroup(
     Masked Float=&MergeHelper_FloatSwitch(0=65, 1=55, 2=55, 3=68, Index=$User.GameModeID, Num=::NewEnumerator2)
     Passthrough Float=&MergeHelper_FloatSwitch(0=60, 1=58, 2=58, 3=70, Index=$User.GameModeID, Num=::NewEnumerator2)
    )
    UsePositionOffset=true
    Write Material Random=true
   )
   &AddVelocity(Velocity=<-1500 0 0>) {enabled=false}
   /VehicleCosmetics/FX/VC_FXParents/Modules/DLR_ProductPreviews(GameModeID:NiagaraInt32=$System.GameModeID, bPreviewVehicleSpeed:NiagaraBool=$System.bPreviewVehicleSpeed) {module=DLR_CoroRemapping}
  ]
  particleUpdateScript [
   &ParticleState()
   &VC_PaintedVariantsOverride(
    Black=<rgba 0.0625 0.0625 0.0625 1>
    BurntSienna=<rgba 0.1771 0.068 0.0387 1>
    Cobalt=<rgba 0.0479 0.0618 0.799 1>
    Crimson=<rgba 0.799 0.0479 0.0683 1>
    DefaultColor=&ColorFromCurve(ColorCurve=DataInterface() {propertyValues {
       AlphaCurve {keys [<at 0.8218 1 RCIM_Cubic>,<at 0.9223 0>]}
       BlueCurve {keys [<at 0.6367 0.8>]}
       GreenCurve {keys [<at 0.6367 0.8>]}
       RedCurve {keys [<at 0.6367 0.8>]}
      }})
    ForestGreen=<rgba 0.0495 0.3391 0.0641 1>
    Grey=<rgba 0.2087 0.2087 0.2087 1>
    Lime=<rgba 0.4235 0.799 0.0479 1>
    Orange=<rgba 0.799 0.3099 0.0479 1>
    Pink=<rgba 0.799 0.0479 0.5259 1>
    Purple=<rgba 0.3443 0.0278 0.799 1>
    R1 DefaultColor=&ColorFromCurve(ColorCurve=DataInterface() {propertyValues {
       AlphaCurve {keys [<at 0.0151 0>, <at 0.1234 2>, {arriveTangent=-2.4249, interpMode=RCIM_Cubic, leaveTangent=-2.4249, time=0.8218, value=2}, <at 0.9482 0>]}
       BlueCurve {keys [<at 0.6367 0>]}
       GreenCurve {keys [<at 0.6367 0.0454>]}
       RedCurve {keys [<at 0.6367 0.8>]}
      }})
    Saffron=<rgba 0.799 0.6304 0.0764 1>
    SkyBlue=<rgba 0.0479 0.5477 0.799 1>
    TitaniumWhite=<rgba 0.3698 0.3698 0.3698 1>
    Use R1 Painted Overrides=true
   ) {module=VC_PaintedVariantsOverride001}
   &ScaleSpriteSize(
    Scale Factor=&MaskVector2dBySpawnGroup(
     Masked Vector2D=&Vector2DFromCurve(Vector2Curve=DataInterface() {propertyValues {XCurve {keys [<at 0 1>]}, YCurve {keys [<at 0 1>]}}}, ScaleCurve=<1 1>)
     Passthrough Vector2D=&Vector2DFromCurve(
      Vector2Curve=DataInterface() {propertyValues {
        XCurve {defaultValue=0.5, keys [{interpMode=RCIM_Cubic, tangentMode=RCTM_User, time=0, value=1.5}, {tangentMode=RCTM_User, time=1, value=1}]}
        YCurve {defaultValue=0.5, keys [{interpMode=RCIM_Cubic, tangentMode=RCTM_User, time=0, value=1.5}, {tangentMode=RCTM_User, time=1, value=1}]}
       }}
      ScaleCurve=<1 1>
     )
     Spawn Group Mask=0
    )
    Scale Sprite Size Mode=::NewEnumerator2
   )
   &self:ScratchModule_02() {module=ScratchModule_02}
   &ShapeLocation(
    Handle Radius:NiagaraFloat=&MaskFloatBySpawnGroup(
     Masked Float=15
     Passthrough Float=&MaskFloatBySpawnGroup(
      Masked Float=15
      Passthrough Float=&MaskFloatBySpawnGroup(
       Masked Float=0
       Passthrough Float=&MaskFloatBySpawnGroup(Masked Float=0, Passthrough Float=0, Spawn Group Mask=3)
       Spawn Group Mask=2
      )
      Spawn Group Mask=1
     )
    )
    Large Radius=&MaskFloatBySpawnGroup(
     Masked Float=35
     Passthrough Float=&MaskFloatBySpawnGroup(Masked Float=30, Passthrough Float=1, Spawn Group Mask=1)
    )
    MaskSpawnGroup=false
    Offset=&TransformVector(
     Vector=&Lerp_Vector(
      A=<0 0 0>
      B=&MaskVectorBySpawnGroup(
       Masked Vector=&MergeHelper_VectorSwitch(
        0=&RandomRangeVector(Minimum=<-400 0 0>, Maximum=<-450 0 0>)
        1=&RandomRangeVector(Minimum=<-100 0 0>, Maximum=<-150 0 0>)
        2=&RandomRangeVector(Minimum=<-100 0 0>, Maximum=<-150 0 0>)
        3=?
        Index=$System.GameModeID
        Num=::NewEnumerator1
       )
       Passthrough Vector=&MaskVectorBySpawnGroup(
        Masked Vector=&MergeHelper_VectorSwitch(
         0=&RandomRangeVector(Minimum=<-350 0 0>, Maximum=<-400 0 0>)
         1=&RandomRangeVector(Minimum=<-100 0 0>, Maximum=<-150 0 0>)
         2=&RandomRangeVector(Minimum=<-100 0 0>, Maximum=<-150 0 0>)
         3=?
         Index=$System.GameModeID
         Num=::NewEnumerator1
        )
        Passthrough Vector=&MaskVectorBySpawnGroup(
         Masked Vector=&MergeHelper_VectorSwitch(0=<-200 0 0>, 1=<-100 0 0>, 2=<-100 0 0>, 3=?, Index=$System.GameModeID, Num=::NewEnumerator1)
         Passthrough Vector=&MaskVectorBySpawnGroup(
          Masked Vector=&MergeHelper_VectorSwitch(0=<-200 0 0>, 1=<-100 0 0>, 2=<-100 0 0>, 3=?, Index=$System.GameModeID, Num=::NewEnumerator1)
          Passthrough Vector=<0 0 0>
          Spawn Group Mask=3
         )
         Spawn Group Mask=2
        )
        Spawn Group Mask=1
       )
      )
      Alpha=$Particles.NormalizedAge
     )
     Destination Space=::Simulation
     Apply Scale=true
    )
    Offset Mode=::NewEnumerator0
    Override Seed=false
    Rotation Angle:NiagaraFloat=&Multiply_Float(
     A=$Engine.Time
     B=&MaskFloat(Mask Source=1, Mask Parameter=$Particles.PartitionParticles.Partition, Mask=0, Masked Float=-0.95, Passthrough Float=0.95)
    )
    Rotation Mode=::NewEnumerator1
    Shape Origin=$Particles.Initial.Position
    Shape Primitive=::NewEnumerator3
    UseOwnerScale=true
   ) {enabled=false}
   &ShapeLocation(
    MaskSpawnGroup=false
    Offset=&MergeHelper_VectorSwitch(0=<-47.5 0 0>, 1=<-47.5 0 0>, 2=<-47 0 50>, 3=<-200 0 37>, Index=$System.GameModeID, Num=::NewEnumerator2)
    Offset Mode=::NewEnumerator0
    Rotation Angle:NiagaraFloat=&Multiply_Float(
     A=$Engine.Time
     B=&MaskFloat(Mask Source=1, Mask Parameter=0, Mask=0, Masked Float=-0.95, Passthrough Float=0.95)
    )
    Rotation Coordinate Space=::Simulation
    Rotation Mode=::NewEnumerator1
    Shape Primitive=::NewEnumerator3
    Torus Distribution Mode:ENiagaraTorusDistributionMode=::NewEnumerator1
    UseOwnerScale=true
    V Position=2
   ) {enabled=false, module=ShapeLocation001}
   &OffsetPosition(
    Position Offset=&MergeHelper_VectorSwitch(0=<-47.5 0 0>, 1=<-47.5 0 0>, 2=<-47 0 50>, 3=<-200 0 37>, Index=$System.GameModeID, Num=::NewEnumerator2)
    Position Output=1
    Position to Offset:NiagaraPosition=$Engine.Emitter.SimulationPosition
   ) {enabled=false}
   &OffsetPosition(
    Position Offset=&MergeHelper_VectorSwitch(0=<-47.5 0 0>, 1=<-47.5 0 0>, 2=<-47 0 50>, 3=<-200 0 37>, Index=$System.GameModeID, Num=::NewEnumerator2)
    Position Offset Coordinate Space:ENiagaraCoordinateSpace=::World
    Position Output=1
    Position to Offset:NiagaraPosition=&AddVectorToPosition(Position=$User.DTCarOffset, Vector=$System.FinalPosition)
   ) {enabled=false, module=OffsetPosition001}
   &CurlNoiseForce(Bias Noise=true, Noise Quality / Cost=::NewEnumerator3, Noise Strength=2286.9231, Pan Noise=true, Pan Noise Field=<1.7333 0 0>) {enabled=false}
   &SolveForcesAndVelocity(Clamp Velocity=false, Limit Acceleration=false)
  ]
  rendererClasses [&NiagaraSpriteRendererProperties]
  renderers [
   {rendererClass=&NiagaraSpriteRendererProperties}
   {rendererClass=&NiagaraSpriteRendererProperties, rendererIndex=1}
   {rendererClass=&NiagaraSpriteRendererProperties, rendererIndex=2}
  ]
 }
 NE_VC_SingleBoost_Ribbon_Parent {
  emitterUpdateScript [
   &EmitterState()
   &VC_FilteredSocketCount() {enabled=false}
   &VC_RateQualityHandler(Epic Spawn Rate=30, High Spawn Rate=30, Low Spawn Rate=30, Use Custom Rate?=true)
   &VC_SpawnRibbon_v1(SpawnRate=$Emitter.Spawn Rate)
  ]
  particleSpawnScript [
   &VC_PaintedVariantsOverride() {enabled=false}
   &VC_InitializeRibbon_v1(
    Custom Offset=true
    Lifetime=&MergeHelper_FloatSwitch(0=1, 1=1, 2=1, 3=1, Index=$User.GameModeID, Num=::NewEnumerator2)
    Local Vector Offset=&AddVectorToPosition(
     Position=&MergeHelper_VectorSwitch(0=<-20 0 0>, 1=<-20 0 0>, 2=<-20 0 50>, 3=<-180 0 37>, Index=$System.GameModeID, Num=::NewEnumerator2)
     Vector=&AddVectorToPosition(Position=$User.DTCarOffset, Vector=Linked() {linkedVariable {name=User.SingleBoostPos, type=NiagaraPosition}})
    )
    Offset Amount=1
    Ribbon Width=70
   )
   &AddVelocity(Velocity=<-1000 0 50>) {module=AddVelocity001}
   &DLR_ProductPreviews(GameModeID:NiagaraInt32=$System.GameModeID, bPreviewVehicleSpeed:NiagaraBool=$System.bPreviewVehicleSpeed)
  ]
  particleUpdateScript [
   &ParticleState()
   &InheritSourceMovement(
    Applied Position Scale=&VectorFromFloat(Value=&FloatFromCurve(FloatCurve=DataInterface() {propertyValues {Curve {
         defaultValue=1
         keys [
          {arriveTangent=-1.7885, interpMode=RCIM_Cubic, leaveTangent=-1.7885, tangentMode=RCTM_User, time=0.05, value=1}
          {interpMode=RCIM_Cubic, tangentMode=RCTM_User, time=1, value=0}
         ]
        }}}))
    ApplyForce=false
    ApplyPosition=true
    ApplyPositionFalloff=false
   ) {enabled=false, module=InheritSourceMovement001}
   &SolveForcesAndVelocity(Clamp Velocity=false, Limit Acceleration=false)
   &VC_UpdateRibbon_v1()
   /SetVariables(
    Particles.R1_Width:NiagaraFloat=&FloatFromCurve(
     FloatCurve=DataInterface() {propertyValues {Curve {keys [
         {arriveTangent=2.4575, interpMode=RCIM_Cubic, leaveTangent=8.5525, tangentMode=RCTM_Break, time=0.02, value=0.0708}
         {arriveTangent=1.2187, interpMode=RCIM_Cubic, leaveTangent=0.015, tangentMode=RCTM_Break, time=0.075, value=0.3979}
         {arriveTangent=-0.0827, interpMode=RCIM_Cubic, leaveTangent=-0.0827, tangentMode=RCTM_User, time=0.4906, value=0.3569}
         {arriveTangent=-0.2415, interpMode=RCIM_Cubic, tangentMode=RCTM_Break, time=1.0074, value=0.2732}
        ]}}}
     Scale Curve=&MergeHelper_FloatSwitch(0=210, 1=185, 2=185, 3=210, Index=$User.GameModeID, Num=::NewEnumerator2)
    )
    Particles.R2_Width:NiagaraFloat=&FloatFromCurve(
     FloatCurve=DataInterface() {propertyValues {Curve {keys [
         {arriveTangent=2.4575, time=0.0268, value=0.3733}
         {arriveTangent=0.6389, time=0.1547, value=0.3771}
         {arriveTangent=-0.0826, leaveTangent=-0.0826, time=0.5111, value=0.2}
         {interpMode=RCIM_Cubic, tangentMode=RCTM_Break, time=0.9995, value=0.2}
        ]}}}
     Scale Curve=&MergeHelper_FloatSwitch(0=135, 1=115, 2=115, 3=135, Index=$User.GameModeID, Num=::NewEnumerator2)
    )
    Particles.R3_Width:NiagaraFloat=&FloatFromCurve(
     FloatCurve=DataInterface() {propertyValues {Curve {keys [
         {arriveTangent=2.4575, interpMode=RCIM_Cubic, leaveTangent=8.5525, tangentMode=RCTM_Break, time=0.02, value=0.0708}
         {arriveTangent=1.2187, interpMode=RCIM_Cubic, leaveTangent=0.015, tangentMode=RCTM_Break, time=0.075, value=0.3979}
         {arriveTangent=-0.0827, interpMode=RCIM_Cubic, leaveTangent=-0.0827, tangentMode=RCTM_User, time=0.4878, value=0.3959}
         {arriveTangent=-0.2415, interpMode=RCIM_Cubic, tangentMode=RCTM_Break, time=1.0018, value=0.3433}
        ]}}}
     Scale Curve=138
    )
   ) {isSetParameters=true}
   &VC_PaintedVariantsOverride(
    Orange=<rgba 0.6724 0.0883 0.2836 1>
    R1 Black:LinearColor=&ColorFromCurve(ColorCurve=DataInterface() {propertyValues {
       AlphaCurve {keys [<at 0 4 RCIM_Cubic>,<at 0.1045 0.9695>,<at 1 0 RCIM_Cubic>]}
       BlueCurve {keys [<at 0 0.035 RCIM_Cubic>,<at 0.6713 0.0208>]}
       GreenCurve {keys [<at 0 0.035 RCIM_Cubic>,<at 0.6713 0.0208>]}
       RedCurve {keys [<at 0 0.035 RCIM_Cubic>,<at 0.6713 0.0208>]}
      }})
    R1 BurntSienna:LinearColor=&ColorFromCurve(ColorCurve=DataInterface() {propertyValues {
       AlphaCurve {keys [<at 0 4 RCIM_Cubic>,<at 0.2166 2>,<at 1 0 RCIM_Cubic>]}
       BlueCurve {keys [<at 0 0.0284>,<at 0.6066 0.0061>]}
       GreenCurve {keys [<at 0 0.0717>,<at 0.6066 0.0193>]}
       RedCurve {keys [<at 0 0.2292>,<at 0.6066 0.067>]}
      }})
    R1 Cobalt:LinearColor=&ColorFromCurve(ColorCurve=DataInterface() {propertyValues {
       AlphaCurve {keys [<at 0 4 RCIM_Cubic>,<at 0.6217 1>,<at 1 0 RCIM_Cubic>]}
       BlueCurve {keys [<at 0 0.6867 RCIM_Cubic>,<at 0.6282 0.067>]}
       GreenCurve {keys [<at 0 0.0551 RCIM_Cubic>,<at 0.6282 0.0013>]}
       RedCurve {keys [<at 0 0.0429 RCIM_Cubic>,<at 0.6282 0>]}
      }})
    R1 Crimson:LinearColor=&ColorFromCurve(ColorCurve=DataInterface() {propertyValues {
       AlphaCurve {keys [<at 0 4 RCIM_Cubic>,<at 1 0 RCIM_Cubic>]}
       BlueCurve {keys [<at 0.0162 0.118>,<at 0.5441 0.0018>]}
       GreenCurve {keys [<at 0.0162 0.1069>,<at 0.5441 0>]}
       RedCurve {keys [<at 0.0162 0.526>,<at 0.5441 0.067>]}
      }})
    R1 DefaultColor=&ColorFromCurve(ColorCurve=DataInterface() {propertyValues {
       AlphaCurve {keys [<at 0 0.05 RCIM_Cubic>]}
       BlueCurve {keys [<at 0.2825 0.0051>]}
       GreenCurve {keys [<at 0.2825 0.0207>]}
       RedCurve {keys [<at 0.2825 0.2696>]}
      }})
    R1 ForestGreen:LinearColor=&ColorFromCurve(ColorCurve=DataInterface() {propertyValues {
       AlphaCurve {keys [<at 0 4 RCIM_Cubic>,<at 1 0 RCIM_Cubic>]}
       BlueCurve {keys [<at 0.0075 0.0716>,<at 0.5485 0.0091>]}
       GreenCurve {keys [<at 0.0075 0.3438>,<at 0.5485 0.067>]}
       RedCurve {keys [<at 0.0075 0.0573>,<at 0.5485 0.0061>]}
      }})
    R1 Grey:LinearColor=&ColorFromCurve(ColorCurve=DataInterface() {propertyValues {
       AlphaCurve {keys [<at 0 1 RCIM_Cubic>,<at 0.8868 0.9>,<at 0.9795 0 RCIM_Cubic>]}
       BlueCurve {keys [<at 0 0.2527>,<at 0.5743 0.067>]}
       GreenCurve {keys [<at 0 0.2527>,<at 0.5743 0.067>]}
       RedCurve {keys [<at 0 0.2527>,<at 0.5743 0.067>]}
      }})
    R1 Lime:LinearColor=&ColorFromCurve(ColorCurve=DataInterface() {propertyValues {
       AlphaCurve {keys [<at 0 1 RCIM_Cubic>,<at 0.8997 0.9>,<at 1 0 RCIM_Cubic>]}
       BlueCurve {keys [<at 0.014 0.0508>,<at 0.611 0.0061>]}
       GreenCurve {keys [<at 0.014 0.6435>,<at 0.611 0.067>]}
       RedCurve {keys [<at 0.014 0.344>,<at 0.611 0.0362>]}
      }})
    R1 Orange:LinearColor=&ColorFromCurve(ColorCurve=DataInterface() {propertyValues {
       AlphaCurve {keys [<at 0 1 RCIM_Cubic>,<at 0.9213 0.986>,<at 1 0 RCIM_Cubic>]}
       BlueCurve {keys [<at 0 0.0705 RCIM_Cubic>,<at 0.6217 0.0077>]}
       GreenCurve {keys [<at 0 0.357 RCIM_Cubic>,<at 0.6217 0.134>]}
       RedCurve {keys [<at 0 0.892 RCIM_Cubic>,<at 0.6217 0.3698>]}
      }})
    R1 Pink:LinearColor=&ColorFromCurve(ColorCurve=DataInterface() {propertyValues {
       AlphaCurve {keys [<at 0 1 RCIM_Cubic>,<at 0.9277 0.9949>,<at 1 0 RCIM_Cubic>]}
       BlueCurve {keys [<at 0.0334 0.4174>,<at 0.6282 0.1219>]}
       GreenCurve {keys [<at 0.0334 0.0991>,<at 0.6282 0.0166>]}
       RedCurve {keys [<at 0.0334 0.6>,<at 0.6282 0.1823>]}
      }})
    R1 Purple:LinearColor=&ColorFromCurve(ColorCurve=DataInterface() {propertyValues {
       AlphaCurve {keys [<at 0 1 RCIM_Cubic>,<at 0.8976 0.9795>,<at 1 0 RCIM_Cubic>]}
       BlueCurve {keys [<at 0.042 0.552>,<at 0.6239 0.1042>]}
       GreenCurve {keys [<at 0.042 0.0436>,<at 0.6239 0.0095>]}
       RedCurve {keys [<at 0.042 0.2523>,<at 0.6239 0.0483>]}
      }})
    R1 Saffron:LinearColor=&ColorFromCurve(ColorCurve=DataInterface() {propertyValues {
       AlphaCurve {keys [<at 0 1 RCIM_Cubic>,<at 0.9083 0.9979>,<at 1 0 RCIM_Cubic>]}
       BlueCurve {keys [<at 0.0075 0.0752>,<at 0.6174 0.0171>]}
       GreenCurve {keys [<at 0.0075 0.747>,<at 0.6174 0.2554>]}
       RedCurve {keys [<at 0.0075 0.952>,<at 0.6174 0.3281>]}
      }})
    R1 SkyBlue:LinearColor=&ColorFromCurve(ColorCurve=DataInterface() {propertyValues {
       AlphaCurve {keys [<at 0 1 RCIM_Cubic>,<at 0.9428 0.9815>,<at 1 0 RCIM_Cubic>]}
       BlueCurve {keys [<at 0 0.552>,<at 0.6088 0.1127>]}
       GreenCurve {keys [<at 0 0.3834>,<at 0.6088 0.2969>]}
       RedCurve {keys [<at 0 0.0436>,<at 0.6088 0.0014>]}
      }})
    R1 TitaniumWhite:LinearColor=&ColorFromCurve(ColorCurve=DataInterface() {propertyValues {
       AlphaCurve {keys [<at 0 1 RCIM_Cubic>,<at 1 0 RCIM_Cubic>]}
       BlueCurve {keys [<at 0 0.6406>,<at 0.5808 0.2292>]}
       GreenCurve {keys [<at 0 0.6406>,<at 0.5808 0.2292>]}
       RedCurve {keys [<at 0 0.6406>,<at 0.5808 0.2292>]}
      }})
    R2 Black:LinearColor=<rgba 0 0.5972 0.3763 1>
    R2 BurntSienna:LinearColor=<rgba 0 0.5972 0.3763 1>
    R2 Cobalt:LinearColor=<rgba 0 0.5972 0.3763 1>
    R2 Crimson:LinearColor=<rgba 0 0.5972 0.3763 1>
    R2 DefaultAlpha=0.8
    R2 DefaultColor:LinearColor=&ColorFromCurve(ColorCurve=DataInterface() {propertyValues {
       AlphaCurve {keys [<at 0.0449 0.05>]}
       BlueCurve {keys [<at 0.4885 0.0698>]}
       GreenCurve {keys [<at 0.4885 0.4033>]}
       RedCurve {keys [<at 0.4885 0.6094>]}
      }})
    R2 ForestGreen:LinearColor=<rgba 0 0.5972 0.3763 1>
    R2 Grey:LinearColor=<rgba 0 0.5972 0.3763 1>
    R2 Lime:LinearColor=<rgba 0 0.5972 0.3763 1>
    R2 Orange:LinearColor=<rgba 0 0.5972 0.3763 1>
    R2 Pink:LinearColor=<rgba 0 0.5972 0.3763 1>
    R2 Purple:LinearColor=<rgba 0 0.5972 0.3763 1>
    R2 Saffron:LinearColor=<rgba 0 0.5972 0.3763 1>
    R2 SkyBlue:LinearColor=<rgba 0 0.5972 0.3763 1>
    R2 TitaniumWhite:LinearColor=<rgba 0 0.5972 0.3763 1>
    R3 Black:LinearColor=<rgba 0.3095 0 0.0097 0.8196>
    R3 BurntSienna:LinearColor=<rgba 0.3095 0 0.0097 0.8196>
    R3 Cobalt:LinearColor=<rgba 0.3095 0 0.0097 0.8196>
    R3 Crimson:LinearColor=<rgba 0.3095 0 0.0097 0.8196>
    R3 DefaultColor:LinearColor=<rgba 0.0696 0.0029 0 0.5>
    R3 ForestGreen:LinearColor=<rgba 0.3095 0 0.0097 0.8196>
    R3 Grey:LinearColor=<rgba 0.3095 0 0.0097 0.8196>
    R3 Lime:LinearColor=<rgba 0.3095 0 0.0097 0.8196>
    R3 Orange:LinearColor=<rgba 0.3095 0 0.0097 0.8196>
    R3 Pink:LinearColor=<rgba 0.3095 0 0.0097 0.8196>
    R3 Purple:LinearColor=<rgba 0.3095 0 0.0097 0.8196>
    R3 Saffron:LinearColor=<rgba 0.3095 0 0.0097 0.8196>
    R3 SkyBlue:LinearColor=<rgba 0.3095 0 0.0097 0.8196>
    R3 TitaniumWhite:LinearColor=<rgba 0.3095 0 0.0097 0.8196>
    Use R1 Painted Overrides=true
    Use R2 Painted Overrides=true
    Use R3 Painted Overrides=true
   ) {module=VC_PaintedVariantsOverride001}
  ]
  rendererClasses [&NiagaraRibbonRendererProperties]
  renderers [
   {rendererClass=&NiagaraRibbonRendererProperties}
   {rendererClass=&NiagaraRibbonRendererProperties, rendererIndex=1}
   {rendererClass=&NiagaraRibbonRendererProperties, rendererIndex=2}
  ]
 }
 Single_Contour {
  emitterUpdateScript [
   &EmitterState(Life Cycle Mode=::NewEnumerator1, Loop Behavior=::NewEnumerator1, Loop Duration Mode=::NewEnumerator1, UseLoopDelay=false)
   &SpawnRate(SpawnGroup=1, SpawnRate=8, Use Spawn Probability=false) {module=SpawnRate002}
   &SpawnBurst_Instantaneous(Spawn Count=4, Use Spawn Probability=false, UseLoopCountLimit=false)
   &SpawnRate(SpawnRate=50, Use Spawn Probability=false) {enabled=false}
   &SpawnBurst_Instantaneous(Spawn Count=4, Spawn Time=1.15, Use Spawn Probability=false, UseLoopCountLimit=false) {module=SpawnBurst_Instantaneous001}
  ]
  particleSpawnScript [
   /SetVariables(Particles.VisibilityTag=$Emitter.SpawnGroup) {isSetParameters=true, module=SetVariables_#5}
   &VC_PaintedVariantsOverride() {enabled=false}
   &InitializeParticle(
    Color Mode=::NewEnumerator0
    Position Offset=&AddVectorToPosition(Position=<0 0 0>, Vector=<0 0 0>)
    Sprite Rotation Mode=::NewEnumerator2
    Sprite Size Mode=::NewEnumerator3
    UsePositionOffset=false
    Write Material Random=true
   ) {enabled=false}
   &InitializeParticle(
    Lifetime=&MaskFloatBySpawnGroup(
     Masked Float=&MergeHelper_FloatSwitch(0=0.5, 1=0.5, 2=0.5, 3=0.5, Index=$System.GameModeID, Num=::NewEnumerator2)
     Passthrough Float=&MergeHelper_FloatSwitch(0=0.2, 1=0.2, 2=0.2, 3=0.2, Index=$System.GameModeID, Num=::NewEnumerator2)
    )
    Position=&MergeHelper_VectorSwitch(0=<-53.5 0 -4.8258>, 1=<-53.5 0 -4.8258>, 2=<-53.5 0 -4.8258>, 3=<-206.908 0.75 35.62>, Index=$System.GameModeID, Num=::NewEnumerator2)
    Position Mode=::NewEnumerator1
    Position Offset=&AddVectorToPosition(Position=$User.DTCarOffset, Vector=Linked() {linkedVariable {name=User.SingleBoostPos, type=NiagaraPosition}})
    Position Offset Coordinate Space=::Local
    Sprite Rotation Mode=::NewEnumerator0
    Sprite Size Mode=::NewEnumerator3
    Uniform Sprite Size=&MaskFloatBySpawnGroup(
     Masked Float=&MergeHelper_FloatSwitch(0=65, 1=55, 2=55, 3=68, Index=$User.GameModeID, Num=::NewEnumerator2)
     Passthrough Float=&MergeHelper_FloatSwitch(0=60, 1=58, 2=58, 3=70, Index=$User.GameModeID, Num=::NewEnumerator2)
    )
    UsePositionOffset=true
    Write Material Random=true
   ) {module=InitializeParticle001}
   &AddVelocity(Velocity=<-1500 0 0>) {enabled=false}
   &ShapeLocation(
    MaskSpawnGroup=false
    Offset=&MergeHelper_VectorSwitch(0=<10 0 0>, 1=<0 0 0>, 2=<0 0 0>, 3=<0 0 0>, Index=$System.GameModeID, Num=::NewEnumerator2)
    Offset Mode=::NewEnumerator0
    Override Seed=false
    Sphere Radius=10
    UseOwnerScale=true
   )
   /SetVariables(
    Particles.dotposition:NiagaraPosition=&AddVectorToPosition(Position=$Particles.Position, Vector=<10 0 0>)
    Particles.dotsoffset=-5
   ) {isSetParameters=true, module=SetVariables_#6}
  ]
  particleUpdateScript [
   &ParticleState()
   &VC_PaintedVariantsOverride(
    Black=<rgba 0.0625 0.0625 0.0625 1>
    BurntSienna=<rgba 0.1771 0.068 0.0387 1>
    Cobalt=<rgba 0.0479 0.0618 0.799 1>
    Crimson=<rgba 0.799 0.0479 0.0683 1>
    DefaultColor=&ColorFromCurve(ColorCurve=DataInterface() {propertyValues {
       AlphaCurve {keys [<at 0.8218 1 RCIM_Cubic>,<at 0.9223 0>]}
       BlueCurve {keys [<at 0.6367 0.8>]}
       GreenCurve {keys [<at 0.6367 0.8>]}
       RedCurve {keys [<at 0.6367 0.8>]}
      }})
    ForestGreen=<rgba 0.0495 0.3391 0.0641 1>
    Grey=<rgba 0.2087 0.2087 0.2087 1>
    Lime=<rgba 0.4235 0.799 0.0479 1>
    Orange=<rgba 0.799 0.3099 0.0479 1>
    Pink=<rgba 0.799 0.0479 0.5259 1>
    Purple=<rgba 0.3443 0.0278 0.799 1>
    R1 DefaultColor=&ColorFromCurve(ColorCurve=DataInterface() {propertyValues {
       AlphaCurve {keys [<at 0.0151 0>, <at 0.1234 2>, {arriveTangent=-2.4249, interpMode=RCIM_Cubic, leaveTangent=-2.4249, time=0.8218, value=2}, <at 0.9482 0>]}
       BlueCurve {keys [<at 0.6367 0>]}
       GreenCurve {keys [<at 0.6367 0.0454>]}
       RedCurve {keys [<at 0.6367 0.8>]}
      }})
    Saffron=<rgba 0.799 0.6304 0.0764 1>
    SkyBlue=<rgba 0.0479 0.5477 0.799 1>
    TitaniumWhite=<rgba 0.3698 0.3698 0.3698 1>
    Use R1 Painted Overrides=true
   ) {module=VC_PaintedVariantsOverride001}
   &ScaleSpriteSize(
    Scale Factor=&MaskVector2dBySpawnGroup(
     Masked Vector2D=&Vector2DFromCurve(Vector2Curve=DataInterface() {propertyValues {XCurve {keys [<at 0 1>]}, YCurve {keys [<at 0 1>]}}}, ScaleCurve=<1 1>)
     Passthrough Vector2D=&Vector2DFromCurve(
      Vector2Curve=DataInterface() {propertyValues {
        XCurve {defaultValue=0.5, keys [{interpMode=RCIM_Cubic, tangentMode=RCTM_User, time=0, value=1.5}, {tangentMode=RCTM_User, time=1, value=1}]}
        YCurve {defaultValue=0.5, keys [{interpMode=RCIM_Cubic, tangentMode=RCTM_User, time=0, value=1.5}, {tangentMode=RCTM_User, time=1, value=1}]}
       }}
      ScaleCurve=<1 1>
     )
     Spawn Group Mask=0
    )
    Scale Sprite Size Mode=::NewEnumerator2
   )
   &SolveForcesAndVelocity(Clamp Velocity=false, Limit Acceleration=false) {enabled=false}
   &DynamicMaterialParameters(
    Index 3 Param 1=&Cosine(
     Normalized Angle=&Add_Float(
      A=$Particles.NormalizedAge
      B=&SelectFloatFromArray(Direct Array Index=0, Float Selection Array=DataInterface() {propertyValues {FloatData [0,0.8]}})
     )
     Period=&MergeHelper_FloatSwitch(0=1, 1=0.6, 2=2, 3=0, Index=$User.GameModeID, Num=::NewEnumerator2)
     Scale=&MergeHelper_FloatSwitch(0=3, 1=10, 2=5, 3=0, Index=$User.GameModeID, Num=::NewEnumerator2)
     Bias=0
    )
    Index 3 Param 2=&Cosine(
     Normalized Angle=&Add_Float(
      A=$Particles.NormalizedAge
      B=&SelectFloatFromArray(Direct Array Index=0, Float Selection Array=DataInterface() {propertyValues {FloatData [0,0.2]}})
     )
     Period=&MergeHelper_FloatSwitch(0=1, 1=1, 2=2, 3=0, Index=$User.GameModeID, Num=::NewEnumerator2)
     Scale=&MergeHelper_FloatSwitch(0=2, 1=1, 2=5, 3=0, Index=$User.GameModeID, Num=::NewEnumerator2)
     Bias=0
    )
    Index 3 Param 3=&Sine(
     Normalized Angle=&Add_Float(
      A=$Particles.NormalizedAge
      B=&SelectFloatFromArray(Direct Array Index=0, Float Selection Array=DataInterface() {propertyValues {FloatData [0,0.5]}})
     )
     Period=&MergeHelper_FloatSwitch(0=1, 1=0.6, 2=2, 3=0, Index=$User.GameModeID, Num=::NewEnumerator2)
     Scale=&MergeHelper_FloatSwitch(0=2, 1=4, 2=5, 3=0, Index=$User.GameModeID, Num=::NewEnumerator2)
     Bias=0
    )
    Module.Index3Float0Write=true
    Module.Index3Float1Write=true
    Module.Index3Float2Write=true
    Module.Index3Float3Write=false
    Write Parameter Index 0=::NewEnumerator0
    Write Parameter Index 3=::NewEnumerator1
   )
  ]
  rendererClasses [&NiagaraSpriteRendererProperties]
  renderers [
   {rendererClass=&NiagaraSpriteRendererProperties}
   {rendererClass=&NiagaraSpriteRendererProperties, rendererIndex=1}
  ]
 }
 Single_Eye {
  emitterUpdateScript [
   &EmitterState(Life Cycle Mode=::NewEnumerator1, Loop Behavior=::NewEnumerator1, Loop Duration Mode=::NewEnumerator1, UseLoopDelay=false)
   &SpawnRate(SpawnGroup=1, SpawnRate=2, Use Spawn Probability=false) {module=SpawnRate002}
   &SpawnBurst_Instantaneous(Spawn Count=4, Use Spawn Probability=false, UseLoopCountLimit=false)
   &SpawnRate(SpawnRate=50, Use Spawn Probability=false) {enabled=false}
  ]
  particleSpawnScript [
   /SetVariables(Particles.VisibilityTag=$Emitter.SpawnGroup) {isSetParameters=true, module=SetVariables_#5}
   &VC_PaintedVariantsOverride() {enabled=false}
   &InitializeParticle(
    Color Mode=::NewEnumerator0
    Position Offset=&AddVectorToPosition(Position=<0 0 0>, Vector=<0 0 0>)
    Sprite Rotation Mode=::NewEnumerator2
    Sprite Size Mode=::NewEnumerator3
    UsePositionOffset=false
    Write Material Random=true
   ) {enabled=false}
   &InitializeParticle(
    Lifetime=&MaskFloatBySpawnGroup(
     Masked Float=&MergeHelper_FloatSwitch(0=0.5, 1=0.5, 2=0.5, 3=0.5, Index=$System.GameModeID, Num=::NewEnumerator2)
     Passthrough Float=&MergeHelper_FloatSwitch(0=0.2, 1=0.2, 2=0.2, 3=0.2, Index=$System.GameModeID, Num=::NewEnumerator2)
    )
    Position=&MergeHelper_VectorSwitch(0=<-53.5 0 -4.8258>, 1=<-53.5 0 -4.8258>, 2=<-53.5 0 -4.8258>, 3=<-206.908 0.75 35.62>, Index=$System.GameModeID, Num=::NewEnumerator2)
    Position Mode=::NewEnumerator1
    Position Offset=&AddVectorToPosition(Position=$User.DTCarOffset, Vector=Linked() {linkedVariable {name=User.SingleBoostPos, type=NiagaraPosition}})
    Position Offset Coordinate Space=::Local
    Sprite Rotation Mode=::NewEnumerator0
    Sprite Size Mode=::NewEnumerator3
    Uniform Sprite Size=&MaskFloatBySpawnGroup(
     Masked Float=&MergeHelper_FloatSwitch(0=65, 1=55, 2=55, 3=68, Index=$User.GameModeID, Num=::NewEnumerator2)
     Passthrough Float=&MergeHelper_FloatSwitch(0=60, 1=58, 2=58, 3=70, Index=$User.GameModeID, Num=::NewEnumerator2)
    )
    UsePositionOffset=true
    Write Material Random=true
   ) {module=InitializeParticle001}
   &AddVelocity(Velocity=<-1500 0 0>) {enabled=false}
   /SetVariables(
    Particles.dotposition:NiagaraPosition=&AddVectorToPosition(Position=$Particles.Position, Vector=<10 0 0>)
    Particles.dotsoffset=-5
   ) {isSetParameters=true, module=SetVariables_#6}
  ]
  particleUpdateScript [
   &ParticleState(Kill Particles When Lifetime Has Elapsed=false, Let Infinitely Lived Particles Die When Emitter Deactivates=true)
   &VC_PaintedVariantsOverride(
    Black=<rgba 0.0625 0.0625 0.0625 1>
    BurntSienna=<rgba 0.1771 0.068 0.0387 1>
    Cobalt=<rgba 0.0479 0.0618 0.799 1>
    Crimson=<rgba 0.799 0.0479 0.0683 1>
    DefaultColor=&ColorFromCurve(ColorCurve=DataInterface() {propertyValues {
       AlphaCurve {keys [<at 0.8218 1 RCIM_Cubic>,<at 0.9223 0>]}
       BlueCurve {keys [<at 0.6367 0.8>]}
       GreenCurve {keys [<at 0.6367 0.8>]}
       RedCurve {keys [<at 0.6367 0.8>]}
      }})
    ForestGreen=<rgba 0.0495 0.3391 0.0641 1>
    Grey=<rgba 0.2087 0.2087 0.2087 1>
    Lime=<rgba 0.4235 0.799 0.0479 1>
    Orange=<rgba 0.799 0.3099 0.0479 1>
    Pink=<rgba 0.799 0.0479 0.5259 1>
    Purple=<rgba 0.3443 0.0278 0.799 1>
    R1 DefaultColor=&ColorFromCurve(ColorCurve=DataInterface() {propertyValues {
       AlphaCurve {keys [<at 0.0151 0>, <at 0.1234 2>, {arriveTangent=-2.4249, interpMode=RCIM_Cubic, leaveTangent=-2.4249, time=0.8218, value=2}, <at 0.9482 0>]}
       BlueCurve {keys [<at 0.6367 0>]}
       GreenCurve {keys [<at 0.6367 0.0454>]}
       RedCurve {keys [<at 0.6367 0.8>]}
      }})
    Saffron=<rgba 0.799 0.6304 0.0764 1>
    SkyBlue=<rgba 0.0479 0.5477 0.799 1>
    TitaniumWhite=<rgba 0.3698 0.3698 0.3698 1>
    Use R1 Painted Overrides=true
   ) {module=VC_PaintedVariantsOverride001}
   &ScaleSpriteSize(
    Scale Factor=&MaskVector2dBySpawnGroup(
     Masked Vector2D=&Vector2DFromCurve(Vector2Curve=DataInterface() {propertyValues {XCurve {keys [<at 0 1>]}, YCurve {keys [<at 0 1>]}}}, ScaleCurve=<1 1>)
     Passthrough Vector2D=&Vector2DFromCurve(
      Vector2Curve=DataInterface() {propertyValues {
        XCurve {defaultValue=0.5, keys [{interpMode=RCIM_Cubic, tangentMode=RCTM_User, time=0, value=1.5}, {tangentMode=RCTM_User, time=1, value=1}]}
        YCurve {defaultValue=0.5, keys [{interpMode=RCIM_Cubic, tangentMode=RCTM_User, time=0, value=1.5}, {tangentMode=RCTM_User, time=1, value=1}]}
       }}
      ScaleCurve=<1 1>
     )
     Spawn Group Mask=0
    )
    Scale Sprite Size Mode=::NewEnumerator2
   )
   &SolveForcesAndVelocity(Clamp Velocity=false, Limit Acceleration=false) {enabled=false}
   &DynamicMaterialParameters(
    Index 3 Param 1=&Cosine(
     Normalized Angle=&Add_Float(
      A=$Particles.NormalizedAge
      B=&SelectFloatFromArray(Direct Array Index=0, Float Selection Array=DataInterface() {propertyValues {FloatData [0,0.8]}})
     )
     Period=&MergeHelper_FloatSwitch(0=6, 1=6, 2=6, 3=6, Index=$User.GameModeID, Num=::NewEnumerator2)
     Scale=&MergeHelper_FloatSwitch(0=2, 1=2, 2=2, 3=1, Index=$User.GameModeID, Num=::NewEnumerator2)
     Bias=5
    )
    Index 3 Param 2=&Cosine(
     Normalized Angle=&Add_Float(
      A=$Particles.NormalizedAge
      B=&SelectFloatFromArray(Direct Array Index=0, Float Selection Array=DataInterface() {propertyValues {FloatData [0,0.2]}})
     )
     Period=&MergeHelper_FloatSwitch(0=6, 1=6, 2=6, 3=6, Index=$User.GameModeID, Num=::NewEnumerator2)
     Scale=&MergeHelper_FloatSwitch(0=2, 1=2, 2=2, 3=1, Index=$User.GameModeID, Num=::NewEnumerator2)
     Bias=5
    )
    Index 3 Param 3=&Sine(
     Normalized Angle=&Add_Float(
      A=$Particles.NormalizedAge
      B=&SelectFloatFromArray(Direct Array Index=0, Float Selection Array=DataInterface() {propertyValues {FloatData [0,0.5]}})
     )
     Period=&MergeHelper_FloatSwitch(0=3, 1=3, 2=3, 3=3, Index=$User.GameModeID, Num=::NewEnumerator2)
     Scale=&MergeHelper_FloatSwitch(0=2, 1=2, 2=2, 3=1, Index=$User.GameModeID, Num=::NewEnumerator2)
     Bias=5
    )
    Module.Index3Float0Write=true
    Module.Index3Float1Write=true
    Module.Index3Float2Write=true
    Module.Index3Float3Write=false
    Write Parameter Index 0=::NewEnumerator0
    Write Parameter Index 3=::NewEnumerator1
   )
  ]
  rendererClasses [&NiagaraSpriteRendererProperties]
  renderers [{rendererClass=&NiagaraSpriteRendererProperties}]
 }
 Smoke_Single {
  emitterUpdateScript [
   &EmitterState(Life Cycle Mode=::NewEnumerator1, Loop Behavior=::NewEnumerator1, Loop Duration Mode=::NewEnumerator1, UseLoopDelay=false)
   &SpawnRate(SpawnRate=6, Use Spawn Probability=false)
  ]
  particleSpawnScript [
   /SetVariables(Particles.VisibilityTag=$Emitter.SpawnGroup) {isSetParameters=true, module=SetVariables_#4}
   &VC_PaintedVariantsOverride() {enabled=false}
   &InitializeParticle(
    Lifetime=&MergeHelper_FloatSwitch(
     0=&MaskFloatBySpawnGroup(Masked Float=1, Passthrough Float=1)
     1=&MaskFloatBySpawnGroup(Masked Float=1, Passthrough Float=1)
     2=&MaskFloatBySpawnGroup(Masked Float=1, Passthrough Float=1)
     3=&MaskFloatBySpawnGroup(Masked Float=1, Passthrough Float=1)
     Index=$User.GameModeID
     Num=::NewEnumerator2
    )
    Position=&MergeHelper_VectorSwitch(0=<-47 0 0>, 1=<-47 0 0>, 2=<0 0 50>, 3=<-200 0 37>, Index=$System.GameModeID, Num=::NewEnumerator2)
    Position Mode=::NewEnumerator1
    Position Offset=&AddVectorToPosition(Position=$User.DTCarOffset, Vector=Linked() {linkedVariable {name=User.SingleBoostPos, type=NiagaraPosition}})
    Position Offset Coordinate Space=::Local
    Sprite Rotation Angle=&RandomRangeFloat(Minimum=-0.2, Maximum=0.35)
    Sprite Rotation Mode=::NewEnumerator1
    Sprite Size Mode=::NewEnumerator3
    Uniform Sprite Size=&MergeHelper_FloatSwitch(
     0=&MaskFloatBySpawnGroup(Masked Float=172, Passthrough Float=158)
     1=&MaskFloatBySpawnGroup(Masked Float=155, Passthrough Float=140)
     2=&MaskFloatBySpawnGroup(Masked Float=155, Passthrough Float=140)
     3=&MaskFloatBySpawnGroup(Masked Float=205, Passthrough Float=185)
     Index=$User.GameModeID
     Num=::NewEnumerator2
    )
    UsePositionOffset=true
    Write Material Random=true
   )
   &AddVelocity(Velocity=<-1500 0 0>) {enabled=false}
   &DLR_ProductPreviews(GameModeID:NiagaraInt32=$System.GameModeID, bPreviewVehicleSpeed:NiagaraBool=$System.bPreviewVehicleSpeed) {enabled=false, module=DLR_CoroRemapping}
  ]
  particleUpdateScript [
   &ParticleState()
   &VC_PaintedVariantsOverride(
    Black=&ColorFromCurve(ColorCurve=DataInterface() {propertyValues {
       AlphaCurve {keys [<at 0 0 RCIM_Cubic>,<at 0.5769 1.5 RCIM_Cubic>]}
       BlueCurve {keys [<at 0.1205 0.014 RCIM_Cubic>]}
       GreenCurve {keys [<at 0.1205 0.014 RCIM_Cubic>]}
       RedCurve {keys [<at 0.1205 0.014 RCIM_Cubic>]}
      }})
    BurntSienna=&ColorFromCurve(ColorCurve=DataInterface() {propertyValues {
       AlphaCurve {keys [<at 0.0646 0>,<at 0.1641 1 RCIM_Cubic>,<at 0.84 0>]}
       BlueCurve {keys [<at 0.1749 0.0058 RCIM_Cubic>,<at 0.5744 0.0014 RCIM_Cubic>]}
       GreenCurve {keys [<at 0.1749 0.0293 RCIM_Cubic>,<at 0.5744 0.025 RCIM_Cubic>]}
       RedCurve {keys [<at 0.1749 0.18 RCIM_Cubic>,<at 0.5744 0.0207 RCIM_Cubic>]}
      }})
    Cobalt=&ColorFromCurve(ColorCurve=DataInterface() {propertyValues {
       AlphaCurve {keys [<at 0 0 RCIM_Cubic>,<at 0.5462 1 RCIM_Cubic>,<at 0.8973 0>]}
       BlueCurve {keys [<at 0.1503 2 RCIM_Cubic>,<at 0.5806 0.08>]}
       GreenCurve {keys [<at 0.1503 0 RCIM_Cubic>,<at 0.5806 0.0216>]}
       RedCurve {keys [<at 0.1503 0.0006 RCIM_Cubic>,<at 0.5806 0.06>]}
      }})
    Crimson=&ColorFromCurve(ColorCurve=DataInterface() {propertyValues {
       AlphaCurve {keys [<at 0.0654 0>,<at 0.1921 1 RCIM_Cubic>,<at 0.8687 0>]}
       BlueCurve {keys [<at 0.2019 0.01 RCIM_Cubic>,<at 0.5683 0.02 RCIM_Cubic>]}
       GreenCurve {keys [<at 0.2019 0.01 RCIM_Cubic>,<at 0.5683 0.04 RCIM_Cubic>]}
       RedCurve {keys [<at 0.2019 0.2 RCIM_Cubic>,<at 0.5683 0.038 RCIM_Cubic>]}
      }})
    DefaultColor=&ColorFromCurve(ColorCurve=DataInterface() {propertyValues {
       AlphaCurve {defaultValue=0, keys [<at 0.7335 5>,<at 0.8894 0>]}
       BlueCurve {keys [{arriveTangent=0.0068, interpMode=RCIM_Cubic, time=0.4311, value=0.0097}, <at 0.6815 0.0046>]}
       GreenCurve {keys [{arriveTangent=-0.3574, interpMode=RCIM_Cubic, time=0.4311, value=0.0395}, <at 0.6815 0.0182>]}
       RedCurve {keys [{arriveTangent=-1.2278, interpMode=RCIM_Cubic, time=0.4311, value=0.314}, <at 0.6815 0.1458>]}
      }})
    ForestGreen=&ColorFromCurve(ColorCurve=DataInterface() {propertyValues {
       AlphaCurve {keys [<at 0.0381 0>,<at 0.1921 1.3 RCIM_Cubic>,<at 0.8655 0>]}
       BlueCurve {keys [<at 0.1872 0 RCIM_Cubic>,<at 0.5708 0.0333>,<at 0.679 0.0373>]}
       GreenCurve {keys [<at 0.1872 0.28 RCIM_Cubic>,<at 0.5708 0.04>,<at 0.679 0.04>]}
       RedCurve {keys [<at 0.1872 0.0467 RCIM_Cubic>,<at 0.5708 0>,<at 0.679 0.024>]}
      }})
    Grey=&ColorFromCurve(ColorCurve=DataInterface() {propertyValues {
       AlphaCurve {keys [
         <at 0 0 RCIM_Cubic>
         {arriveTangent=1.4839, interpMode=RCIM_Cubic, leaveTangent=1.4839, time=0.4899, value=2.5}
         <at 0.6065 0.9>
         <at 0.9085 0>
        ]}
       BlueCurve {keys [<at 0.3574 0.12 RCIM_Cubic>,<at 0.5955 0.035>]}
       GreenCurve {keys [<at 0.3574 0.12 RCIM_Cubic>,<at 0.5955 0.035>]}
       RedCurve {keys [<at 0.3574 0.12 RCIM_Cubic>,<at 0.5955 0.035>]}
      }})
    Lime=&ColorFromCurve(ColorCurve=DataInterface() {propertyValues {
       AlphaCurve {keys [
         <at 0.082 0 RCIM_Cubic>
         {arriveTangent=2.8732, interpMode=RCIM_Cubic, leaveTangent=2.8732, time=0.1859, value=1}
         <at 0.4301 1>
         <at 0.754 0>
        ]}
       BlueCurve {keys [<at 0.2222 0 RCIM_Cubic>,<at 0.5391 0.061 RCIM_Cubic>]}
       GreenCurve {keys [<at 0.2222 0.7917 RCIM_Cubic>,<at 0.5391 0.1476 RCIM_Cubic>]}
       RedCurve {keys [<at 0.2222 0.1696 RCIM_Cubic>,<at 0.5391 0.0081 RCIM_Cubic>]}
      }})
    Orange=&ColorFromCurve(ColorCurve=DataInterface() {propertyValues {
       AlphaCurve {keys [
         <at 0.0486 0 RCIM_Cubic>
         {arriveTangent=2.6214, interpMode=RCIM_Cubic, leaveTangent=2.6214, time=0.1859, value=1}
         <at 0.4301 1>
         <at 0.7524 0>
        ]}
       BlueCurve {keys [<at 0.1828 0.0008 RCIM_Cubic>,<at 0.5088 0 RCIM_Cubic>]}
       GreenCurve {keys [<at 0.1828 0.2332 RCIM_Cubic>,<at 0.5088 0.09 RCIM_Cubic>]}
       RedCurve {keys [<at 0.1828 0.6146 RCIM_Cubic>,<at 0.5088 0.075 RCIM_Cubic>]}
      }})
    Pink=&ColorFromCurve(ColorCurve=DataInterface() {propertyValues {
       AlphaCurve {keys [
         <at 0.0629 0 RCIM_Cubic>
         {arriveTangent=2.7237, interpMode=RCIM_Cubic, leaveTangent=2.7237, time=0.1859, value=1}
         <at 0.4301 1>
         <at 0.8273 0>
        ]}
       BlueCurve {keys [<at 0.1828 0.2344 RCIM_Cubic>,<at 0.5741 0.04 RCIM_Cubic>]}
       GreenCurve {keys [<at 0.1828 0.093 RCIM_Cubic>,<at 0.5741 0.06 RCIM_Cubic>]}
       RedCurve {keys [<at 0.1828 0.3 RCIM_Cubic>,<at 0.5741 0.1 RCIM_Cubic>]}
      }})
    Purple=&ColorFromCurve(ColorCurve=DataInterface() {propertyValues {
       AlphaCurve {keys [<at 0.0508 0 RCIM_Cubic>,<at 0.2922 1 RCIM_Cubic>,<at 0.949 0>]}
       BlueCurve {keys [<at 0.2333 0.2 RCIM_Cubic>,<at 0.6077 0.0365 RCIM_Cubic>]}
       GreenCurve {keys [<at 0.2333 0.01 RCIM_Cubic>,<at 0.6077 0.0043 RCIM_Cubic>]}
       RedCurve {keys [<at 0.2333 0.1089 RCIM_Cubic>,<at 0.6077 0.0214 RCIM_Cubic>]}
      }})
    R1 Black:LinearColor=&ColorFromCurve(ColorCurve=DataInterface() {propertyValues {
       AlphaCurve {keys [<at 0.1097 0 RCIM_Cubic>,<at 0.2745 1 RCIM_Cubic>,<at 0.6836 0>]}
       BlueCurve {keys [<at 0.1091 0.01 RCIM_Cubic>,<at 0.2892 0.1771 RCIM_Cubic>]}
       GreenCurve {keys [<at 0.1091 0.01 RCIM_Cubic>,<at 0.2892 0.1771 RCIM_Cubic>]}
       RedCurve {keys [<at 0.1091 0.01 RCIM_Cubic>,<at 0.2892 0.1771 RCIM_Cubic>]}
      }})
    R1 BurntSienna:LinearColor=&ColorFromCurve(ColorCurve=DataInterface() {propertyValues {
       AlphaCurve {keys [<at 0 0>,<at 0.469 1 RCIM_Cubic>,<at 0.6525 0 RCIM_Cubic>]}
       BlueCurve {keys [<at 0 0.0178 RCIM_Cubic>,<at 0.6548 0.0108 RCIM_Cubic>]}
       GreenCurve {keys [<at 0 0.0908 RCIM_Cubic>,<at 0.6548 0.6927 RCIM_Cubic>]}
       RedCurve {keys [<at 0 0.5573 RCIM_Cubic>,<at 0.6548 0.57 RCIM_Cubic>]}
      }})
    R1 Cobalt:LinearColor=&ColorFromCurve(ColorCurve=DataInterface() {propertyValues {
       AlphaCurve {keys [<at 0.1174 0 RCIM_Cubic>,<at 0.2528 1 RCIM_Cubic>,<at 0.6105 0>]}
       BlueCurve {keys [<at 0.1753 1 RCIM_Cubic>,<at 0.6077 3>]}
       GreenCurve {keys [<at 0.1753 0.4069 RCIM_Cubic>,<at 0.6077 0>]}
       RedCurve {keys [<at 0.1753 0.3646 RCIM_Cubic>,<at 0.6077 2.5078>]}
      }})
    R1 Crimson:LinearColor=&ColorFromCurve(ColorCurve=DataInterface() {propertyValues {
       AlphaCurve {keys [
         <at 0.147 0 RCIM_Cubic>
         {arriveTangent=3.9441, interpMode=RCIM_Cubic, leaveTangent=3.9441, time=0.2808, value=1}
         <at 0.4005 1>
         <at 0.6074 0>
        ]}
       BlueCurve {keys [<at 0.1753 0.1286 RCIM_Cubic>,<at 0.6077 0>]}
       GreenCurve {keys [<at 0.1753 0.1763 RCIM_Cubic>,<at 0.6077 1.4989>]}
       RedCurve {keys [<at 0.1753 1 RCIM_Cubic>,<at 0.6077 1.5>]}
      }})
    R1 DefaultColor=&ColorFromCurve(ColorCurve=DataInterface() {propertyValues {
       AlphaCurve {defaultValue=0, keys [<at 0.7335 10>,<at 0.8894 0>]}
       BlueCurve {keys [<at 0.4311 0 RCIM_Cubic>]}
       GreenCurve {keys [<at 0.4311 0 RCIM_Cubic>]}
       RedCurve {keys [<at 0.4311 0 RCIM_Cubic>]}
      }})
    R1 ForestGreen:LinearColor=&ColorFromCurve(ColorCurve=DataInterface() {propertyValues {
       AlphaCurve {keys [<at 0.0739 0 RCIM_Cubic>, {arriveTangent=3.9441, interpMode=RCIM_Cubic, time=0.2808, value=1}, <at 0.4005 1>, <at 0.6074 0>]}
       BlueCurve {keys [{arriveTangent=3.2911, interpMode=RCIM_Cubic, time=0.1753, value=0}, <at 0.6077 2>]}
       GreenCurve {keys [{arriveTangent=2.7598, interpMode=RCIM_Cubic, time=0.1753, value=0.8594}, <at 0.6077 2>]}
       RedCurve {keys [{arriveTangent=-0.1276, interpMode=RCIM_Cubic, time=0.1753, value=0.2064}, <at 0.6077 0>]}
      }})
    R1 Grey:LinearColor=&ColorFromCurve(ColorCurve=DataInterface() {propertyValues {
       AlphaCurve {keys [<at 0.1143 0>,<at 0.3088 1 RCIM_Cubic>,<at 0.6043 0 RCIM_Cubic>]}
       BlueCurve {keys [<at 0.3072 0.6562 RCIM_Cubic>]}
       GreenCurve {keys [<at 0.3072 0.6562 RCIM_Cubic>]}
       RedCurve {keys [<at 0.3072 0.6562 RCIM_Cubic>]}
      }})
    R1 Lime:LinearColor=&ColorFromCurve(ColorCurve=DataInterface() {propertyValues {
       AlphaCurve {keys [<at 0.0739 0 RCIM_Cubic>, {arriveTangent=3.9441, interpMode=RCIM_Cubic, time=0.2808, value=1}, <at 0.4005 1>, <at 0.6074 0>]}
       BlueCurve {keys [<at 0.091 0 RCIM_Cubic>,<at 0.6077 1.3543>]}
       GreenCurve {keys [<at 0.091 2 RCIM_Cubic>,<at 0.6077 2>]}
       RedCurve {keys [<at 0.091 1.1387 RCIM_Cubic>,<at 0.6077 0>]}
      }})
    R1 Orange:LinearColor=&ColorFromCurve(ColorCurve=DataInterface() {propertyValues {
       AlphaCurve {keys [<at 0.0739 0 RCIM_Cubic>, {arriveTangent=3.9441, interpMode=RCIM_Cubic, time=0.2808, value=1}, <at 0.4005 1>, <at 0.6074 0>]}
       BlueCurve {keys [<at 0.091 0 RCIM_Cubic>,<at 0.6077 0.2815>]}
       GreenCurve {keys [<at 0.091 0.7377 RCIM_Cubic>,<at 0.6077 1>]}
       RedCurve {keys [<at 0.091 2 RCIM_Cubic>,<at 0.6077 0.7462>]}
      }})
    R1 Pink:LinearColor=&ColorFromCurve(ColorCurve=DataInterface() {propertyValues {
       AlphaCurve {keys [<at 0.0739 0 RCIM_Cubic>, {arriveTangent=3.9441, interpMode=RCIM_Cubic, time=0.2808, value=1}, <at 0.4005 1>, <at 0.6074 0>]}
       BlueCurve {keys [<at 0.091 1.7582 RCIM_Cubic>,<at 0.3276 0.4062>]}
       GreenCurve {keys [<at 0.091 0.1064 RCIM_Cubic>,<at 0.3276 0.8298>]}
       RedCurve {keys [<at 0.091 2 RCIM_Cubic>,<at 0.3276 2>]}
      }})
    R1 Purple:LinearColor=&ColorFromCurve(ColorCurve=DataInterface() {propertyValues {
       AlphaCurve {keys [<at 0.1131 0 RCIM_Cubic>, {arriveTangent=3.9441, interpMode=RCIM_Cubic, time=0.2466, value=1}, <at 0.4005 1>, <at 0.6074 0>]}
       BlueCurve {keys [<at 0.2037 5 RCIM_Cubic>,<at 0.3276 0.709>]}
       GreenCurve {keys [<at 0.2037 0 RCIM_Cubic>,<at 0.3276 0.1062>]}
       RedCurve {keys [<at 0.2037 0.9897 RCIM_Cubic>,<at 0.3276 1>]}
      }})
    R1 Saffron:LinearColor=&ColorFromCurve(ColorCurve=DataInterface() {propertyValues {
       AlphaCurve {keys [<at 0.1131 0 RCIM_Cubic>, {arriveTangent=3.9441, interpMode=RCIM_Cubic, time=0.2466, value=1}, <at 0.4005 1>, <at 0.6074 0>]}
       BlueCurve {keys [<at 0.1542 0.875 RCIM_Cubic>,<at 0.3276 0.533>]}
       GreenCurve {keys [<at 0.1542 2.7603 RCIM_Cubic>,<at 0.3276 2>]}
       RedCurve {keys [<at 0.1542 3 RCIM_Cubic>,<at 0.3276 0.9975>]}
      }})
    R1 SkyBlue:LinearColor=&ColorFromCurve(ColorCurve=DataInterface() {propertyValues {
       AlphaCurve {keys [<at 0.1131 0 RCIM_Cubic>, {arriveTangent=3.9441, interpMode=RCIM_Cubic, time=0.2466, value=1}, <at 0.4005 1>, <at 0.6074 0>]}
       BlueCurve {keys [<at 0.1542 2 RCIM_Cubic>,<at 0.3276 1>]}
       GreenCurve {keys [<at 0.1542 1.715 RCIM_Cubic>,<at 0.3276 0.3668>]}
       RedCurve {keys [<at 0.1542 0.0312 RCIM_Cubic>,<at 0.3276 0.519>]}
      }})
    R1 TitaniumWhite:LinearColor=&ColorFromCurve(ColorCurve=DataInterface() {propertyValues {
       AlphaCurve {keys [<at 0.1131 0 RCIM_Cubic>, {arriveTangent=3.9441, interpMode=RCIM_Cubic, time=0.2466, value=1}, <at 0.4005 1>, <at 0.6074 0>]}
       BlueCurve {keys [<at 0.1542 2 RCIM_Cubic>,<at 0.3276 2>]}
       GreenCurve {keys [<at 0.1542 2 RCIM_Cubic>,<at 0.3276 2>]}
       RedCurve {keys [<at 0.1542 2 RCIM_Cubic>,<at 0.3276 2>]}
      }})
    R2 Black:LinearColor=&ColorFromCurve(ColorCurve=DataInterface() {propertyValues {
       AlphaCurve {keys [<at 0.0677 0 RCIM_Cubic>,<at 0.2668 0.8 RCIM_Cubic>,<at 1 0>]}
       BlueCurve {keys [<at 0.1091 0.001 RCIM_Cubic>,<at 0.2892 0.02 RCIM_Cubic>]}
       GreenCurve {keys [<at 0.1091 0.001 RCIM_Cubic>,<at 0.2892 0.02 RCIM_Cubic>]}
       RedCurve {keys [<at 0.1091 0.001 RCIM_Cubic>,<at 0.2892 0.02 RCIM_Cubic>]}
      }})
    R2 BurntSienna:LinearColor=&ColorFromCurve(ColorCurve=DataInterface() {propertyValues {
       AlphaCurve {keys [<at 0.1765 0 RCIM_Cubic>,<at 0.4052 0.6 RCIM_Cubic>]}
       BlueCurve {keys [<at 0.2419 0.0062 RCIM_Cubic>,<at 0.3865 0.0191 RCIM_Cubic>]}
       GreenCurve {keys [<at 0.2419 0.0287 RCIM_Cubic>,<at 0.3865 0.0573 RCIM_Cubic>]}
       RedCurve {keys [<at 0.2419 0.1771 RCIM_Cubic>,<at 0.3865 0.0504 RCIM_Cubic>]}
      }})
    R2 Cobalt:LinearColor=&ColorFromCurve(ColorCurve=DataInterface() {propertyValues {
       AlphaCurve {keys [<at 0.0817 0 RCIM_Cubic>,<at 0.4239 0.6 RCIM_Cubic>,<at 0.8485 0>]}
       BlueCurve {keys [<at 0.1753 0.3281 RCIM_Cubic>,<at 0.4954 0.224>]}
       GreenCurve {keys [<at 0.1753 0.0218 RCIM_Cubic>,<at 0.4954 0>]}
       RedCurve {keys [<at 0.1753 0 RCIM_Cubic>,<at 0.4954 0.1872>]}
      }})
    R2 Crimson:LinearColor=&ColorFromCurve(ColorCurve=DataInterface() {propertyValues {
       AlphaCurve {keys [<at 0.0817 0 RCIM_Cubic>,<at 0.3492 0.8 RCIM_Cubic>,<at 0.8485 0>]}
       BlueCurve {keys [<at 0.1753 0 RCIM_Cubic>,<at 0.4954 0.0137>]}
       GreenCurve {keys [<at 0.1753 0.0039 RCIM_Cubic>,<at 0.4954 0.0729>]}
       RedCurve {keys [<at 0.1753 0.2083 RCIM_Cubic>,<at 0.4954 0.0566>]}
      }})
    R2 DefaultColor:LinearColor=&ColorFromCurve(ColorCurve=DataInterface() {propertyValues {
       AlphaCurve {defaultValue=0, keys [<at 0.2258 0>,<at 0.5147 1>,<at 0.8157 0>]}
       BlueCurve {keys [<at 0.4311 0.1949 RCIM_Cubic>]}
       GreenCurve {keys [<at 0.4311 0.6468 RCIM_Cubic>]}
       RedCurve {keys [<at 0.4311 1 RCIM_Cubic>]}
      }})
    R2 ForestGreen:LinearColor=&ColorFromCurve(ColorCurve=DataInterface() {propertyValues {
       AlphaCurve {keys [
         <at 0.049 0>
         {arriveTangent=1.2606, interpMode=RCIM_Cubic, leaveTangent=1.2606, time=0.4161, value=0.8}
         <at 0.6836 0.8>
         <at 1 0 RCIM_Cubic>
        ]}
       BlueCurve {keys [<at 0 0.0334>,<at 0.4736 0.0922 RCIM_Cubic>]}
       GreenCurve {keys [<at 0 0.1823>,<at 0.4736 0.1198 RCIM_Cubic>]}
       RedCurve {keys [<at 0 0>,<at 0.4736 0 RCIM_Cubic>]}
      }})
    R2 Grey:LinearColor=&ColorFromCurve(ColorCurve=DataInterface() {propertyValues {
       AlphaCurve {keys [
         <at 0.147 0 RCIM_Cubic>
         {arriveTangent=3.1553, interpMode=RCIM_Cubic, leaveTangent=3.1553, time=0.2808, value=0.8}
         <at 0.4005 0.8>
         <at 0.6074 0>
        ]}
       BlueCurve {keys [<at 0.0708 0.2083 RCIM_Cubic>,<at 0.3508 0.0833>]}
       GreenCurve {keys [<at 0.0708 0.2083 RCIM_Cubic>,<at 0.3508 0.0833>]}
       RedCurve {keys [<at 0.0708 0.2083 RCIM_Cubic>,<at 0.3508 0.0833>]}
      }})
    R2 Lime:LinearColor=&ColorFromCurve(ColorCurve=DataInterface() {propertyValues {
       AlphaCurve {keys [
         <at 0.0412 0 RCIM_Cubic>
         {arriveTangent=1.6698, interpMode=RCIM_Cubic, leaveTangent=1.6698, time=0.2808, value=0.6}
         <at 0.4005 0.6>
         <at 0.6074 0>
        ]}
       BlueCurve {keys [<at 0.0941 0.0002 RCIM_Cubic>,<at 0.3803 0.1206>]}
       GreenCurve {keys [<at 0.0941 0.3438 RCIM_Cubic>,<at 0.3803 0.1927>]}
       RedCurve {keys [<at 0.0941 0.2305 RCIM_Cubic>,<at 0.3803 0.0384>]}
      }})
    R2 Orange:LinearColor=&ColorFromCurve(ColorCurve=DataInterface() {propertyValues {
       AlphaCurve {keys [
         <at 0.0412 0 RCIM_Cubic>
         {arriveTangent=1.6698, interpMode=RCIM_Cubic, leaveTangent=1.6698, time=0.1552, value=0.5}
         <at 0.4005 0.6>
         <at 0.6074 0>
        ]}
       BlueCurve {keys [<at 0.1583 0 RCIM_Cubic>,<at 0.3803 0>]}
       GreenCurve {keys [<at 0.1583 0.2274 RCIM_Cubic>,<at 0.3803 0.1458>]}
       RedCurve {keys [<at 0.1583 0.5469 RCIM_Cubic>,<at 0.3803 0.1076>]}
      }})
    R2 Pink:LinearColor=&ColorFromCurve(ColorCurve=DataInterface() {propertyValues {
       AlphaCurve {keys [
         <at 0.0412 0 RCIM_Cubic>
         {arriveTangent=1.6698, interpMode=RCIM_Cubic, leaveTangent=1.6698, time=0.1552, value=0.5}
         <at 0.4005 0.6>
         <at 0.6074 0>
        ]}
       BlueCurve {keys [<at 0.1583 0.2922 RCIM_Cubic>,<at 0.3803 0.0526>]}
       GreenCurve {keys [<at 0.1583 0.0423 RCIM_Cubic>,<at 0.3803 0.0787>]}
       RedCurve {keys [<at 0.1583 0.4271 RCIM_Cubic>,<at 0.3803 0.1094>]}
      }})
    R2 Purple:LinearColor=&ColorFromCurve(ColorCurve=DataInterface() {propertyValues {
       AlphaCurve {keys [
         <at 0.0412 0 RCIM_Cubic>
         {arriveTangent=1.6698, interpMode=RCIM_Cubic, leaveTangent=1.6698, time=0.1552, value=0.5}
         <at 0.4005 0.6>
         <at 0.6074 0>
        ]}
       BlueCurve {keys [<at 0.1583 1 RCIM_Cubic>,<at 0.3803 0.08>]}
       GreenCurve {keys [<at 0.1583 0 RCIM_Cubic>,<at 0.3803 0.0166>]}
       RedCurve {keys [<at 0.1583 0.287 RCIM_Cubic>,<at 0.3803 0.1094>]}
      }})
    R2 Saffron:LinearColor=&ColorFromCurve(ColorCurve=DataInterface() {propertyValues {
       AlphaCurve {keys [
         <at 0.0412 0 RCIM_Cubic>
         {arriveTangent=1.6698, interpMode=RCIM_Cubic, leaveTangent=1.6698, time=0.1552, value=0.5}
         <at 0.4005 0.6>
         <at 0.6074 0>
        ]}
       BlueCurve {keys [<at 0.1583 0 RCIM_Cubic>,<at 0.3803 0.0873>]}
       GreenCurve {keys [<at 0.1583 0.8117 RCIM_Cubic>,<at 0.3803 0.2396>]}
       RedCurve {keys [<at 0.1583 0.875 RCIM_Cubic>,<at 0.3803 0.1074>]}
      }})
    R2 SkyBlue:LinearColor=&ColorFromCurve(ColorCurve=DataInterface() {propertyValues {
       AlphaCurve {keys [
         <at 0.0412 0 RCIM_Cubic>
         {arriveTangent=1.6698, interpMode=RCIM_Cubic, leaveTangent=1.6698, time=0.1552, value=0.5}
         <at 0.4005 0.6>
         <at 0.6074 0>
        ]}
       BlueCurve {keys [<at 0.1583 0.4167 RCIM_Cubic>,<at 0.3803 0.125>]}
       GreenCurve {keys [<at 0.1583 0.2756 RCIM_Cubic>,<at 0.3803 0.041>]}
       RedCurve {keys [<at 0.1583 0.0434 RCIM_Cubic>,<at 0.3803 0.0705>]}
      }})
    R2 TitaniumWhite:LinearColor=&ColorFromCurve(ColorCurve=DataInterface() {propertyValues {
       AlphaCurve {keys [
         <at 0.0412 0 RCIM_Cubic>
         {arriveTangent=1.6698, interpMode=RCIM_Cubic, leaveTangent=1.6698, time=0.1552, value=0.5}
         <at 0.4005 0.6>
         <at 0.6074 0>
        ]}
       BlueCurve {keys [<at 0.1583 0.6667 RCIM_Cubic>,<at 0.3803 0.2031>]}
       GreenCurve {keys [<at 0.1583 0.6667 RCIM_Cubic>,<at 0.3803 0.2031>]}
       RedCurve {keys [<at 0.1583 0.6667 RCIM_Cubic>,<at 0.3803 0.2031>]}
      }})
    R3 DefaultColor:LinearColor=&ColorFromCurve(ColorCurve=DataInterface() {propertyValues {
       AlphaCurve {defaultValue=0, keys [<at 0.7335 5>,<at 0.8894 0>]}
       BlueCurve {keys [{arriveTangent=0.0068, interpMode=RCIM_Cubic, time=0.4311, value=0.0097}, <at 0.6815 0.0046>]}
       GreenCurve {keys [{arriveTangent=-0.3574, interpMode=RCIM_Cubic, time=0.4311, value=0.0395}, <at 0.6815 0.0182>]}
       RedCurve {keys [{arriveTangent=-1.2278, interpMode=RCIM_Cubic, time=0.4311, value=0.314}, <at 0.6815 0.1458>]}
      }})
    Saffron=&ColorFromCurve(ColorCurve=DataInterface() {propertyValues {
       AlphaCurve {keys [
         <at 0.0789 0 RCIM_Cubic>
         {arriveTangent=2.7665, interpMode=RCIM_Cubic, leaveTangent=2.7665, time=0.2317, value=1}
         <at 0.4403 1>
         <at 0.7668 0>
        ]}
       BlueCurve {keys [<at 0.2062 0 RCIM_Cubic>,<at 0.5327 0 RCIM_Cubic>]}
       GreenCurve {keys [<at 0.2062 0.7424 RCIM_Cubic>,<at 0.5327 0.15 RCIM_Cubic>]}
       RedCurve {keys [<at 0.2062 0.875 RCIM_Cubic>,<at 0.5327 0.0779 RCIM_Cubic>]}
      }})
    SkyBlue=&ColorFromCurve(ColorCurve=DataInterface() {propertyValues {
       AlphaCurve {keys [<at 0.0789 0 RCIM_Cubic>, {arriveTangent=2.5883, interpMode=RCIM_Cubic, time=0.2349, value=0.8}, <at 0.4301 1>, <at 0.8209 0>]}
       BlueCurve {keys [<at 0.1022 0.4792 RCIM_Cubic>,<at 0.5215 0.1 RCIM_Cubic>]}
       GreenCurve {keys [<at 0.1022 0.4551 RCIM_Cubic>,<at 0.5215 0.018 RCIM_Cubic>]}
       RedCurve {keys [<at 0.1022 0.0393 RCIM_Cubic>,<at 0.5215 0.0466 RCIM_Cubic>]}
      }})
    TitaniumWhite=&ColorFromCurve(ColorCurve=DataInterface() {propertyValues {
       AlphaCurve {keys [
         <at 0.0852 0 RCIM_Cubic>
         {arriveTangent=2.8998, interpMode=RCIM_Cubic, leaveTangent=2.8998, time=0.2078, value=0.5}
         <at 0.4301 1>
         <at 0.6856 0>
        ]}
       BlueCurve {keys [<at 0.1022 1 RCIM_Cubic>,<at 0.5391 0.11 RCIM_Cubic>]}
       GreenCurve {keys [<at 0.1022 1 RCIM_Cubic>,<at 0.5391 0.11 RCIM_Cubic>]}
       RedCurve {keys [<at 0.1022 1 RCIM_Cubic>,<at 0.5391 0.11 RCIM_Cubic>]}
      }})
    Use R1 Painted Overrides=true
    Use R2 Painted Overrides=true
    Use R3 Painted Overrides=true
   ) {module=VC_PaintedVariantsOverride001}
   &ScaleSpriteSize(
    Uniform Curve Scale=1.2
    Uniform Curve Sprite Scale=DataInterface() {propertyValues {Curve {
       defaultValue=5
       keys [
        {
         arriveTangent=-0.0786
         arriveTangentWeight=0.4631
         interpMode=RCIM_Cubic
         leaveTangent=-0.0786
         leaveTangentWeight=0.4631
         tangentMode=RCTM_User
         tangentWeightMode=RCTWM_WeightedBoth
         time=0
         value=1.2
        }
        {
         arriveTangent=1.2246
         arriveTangentWeight=0.5053
         interpMode=RCIM_Cubic
         leaveTangent=1.2246
         leaveTangentWeight=0.5053
         tangentMode=RCTM_User
         tangentWeightMode=RCTWM_WeightedBoth
         time=1
         value=0.85
        }
       ]
      }}}
   )
   &SubUVAnimation(
    End Frame Range Override=64
    SubUV Animation Mode:ENiagara_SubUVLookupModeV2=::NewEnumerator1
    SubUV Curve=DataInterface() {propertyValues {Curve {defaultValue=0, keys [{arriveTangent=0.3386, arriveTangentWeight=0.2442, leaveTangentWeight=0.2442, time=0, value=0}, <at 1 1>]}}}
    UseEndFrame=true
    UseStartFrame=true
   ) {module=SubUVAnimation003}
   /SetVariables(Particles.HighlightColor:LinearColor=&HueShiftLinearColor(
     Hue=0
     Saturation=1
     Value=&FloatFromCurve(
      FloatCurve=DataInterface() {propertyValues {Curve {keys [<at 0 4 RCIM_Constant>,<at 0.07 1>]}}}
      CurveIndex=&RandomRangeFloat(Minimum=0, Maximum=1)
     )
     Linear Color=$Particles.Color
    )) {isSetParameters=true}
   &SolveForcesAndVelocity(Clamp Velocity=false, Limit Acceleration=false)
  ]
  rendererClasses [&NiagaraSpriteRendererProperties]
  renderers [
   {rendererClass=&NiagaraSpriteRendererProperties}
   {rendererClass=&NiagaraSpriteRendererProperties, rendererIndex=1}
  ]
 }
}
summary {
 systemName=NS_EVIL_PROVIS
 emitters [
  {emitterName=NE_VC_SingleBoost_Ribbon_Parent, rendererClasses [&NiagaraRibbonRendererProperties]}
  {emitterName=Smoke_Single, rendererClasses [&NiagaraSpriteRendererProperties]}
  {bEnabled=false, emitterName=Eye_Single, rendererClasses [&NiagaraSpriteRendererProperties]}
  {emitterName=Single_Eye, rendererClasses [&NiagaraSpriteRendererProperties]}
  {emitterName=Single_Contour, rendererClasses [&NiagaraSpriteRendererProperties]}
 ]
 userVariables [
  User.bIsFrontendPreview {type=NiagaraBool}
  User.bIsPainted {type=NiagaraBool}
  User.blsIsCoroColor {type=NiagaraBool}
  User.bPreviewVehicleSpeed {type=NiagaraBool}
  User.ColorVariantID {type=NiagaraInt32}
  User.DTCarOffset {type=NiagaraPosition, description="", defaultValue {}}
  User.GameModeID {type=NiagaraInt32, defaultValue=1}
  User.GreyTColor {type=LinearColor, description="", defaultValue {struct=LinearColor, value=<rgba 1 1 1 1>}}
  User.InnerCColor {type=LinearColor, description="", defaultValue {struct=LinearColor, value=<rgba 1 1 1 1>}}
  User.MidCColor {type=LinearColor, description="", defaultValue {struct=LinearColor, value=<rgba 1 1 1 1>}}
  User.MTrailColor {type=LinearColor, description="", defaultValue {struct=LinearColor, value=<rgba 1 1 1 1>}}
  User.offset {type=Vector3f, description="", defaultValue {struct=Vector3f, value=<0 0 0>}}
  User.PaintedColor {type=LinearColor, description="", defaultValue {struct=LinearColor, value=<rgba 1 1 1 1>}}
  User.PaintedScalar {type=NiagaraFloat, defaultValue=1, description=""}
  User.PortCount {type=NiagaraInt32, defaultValue=1}
  User.SingleBoostPos {type=NiagaraPosition, description="", defaultValue {}}
  User.SndTColor {type=LinearColor, description="", defaultValue {struct=LinearColor, value=<rgba 1 1 1 1>}}
  User.SpiralColor {type=LinearColor, description="", defaultValue {struct=LinearColor, value=<rgba 1 1 1 1>}}
 ]
}
